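# Idle Niagara effect for the StrategicSpur cosmetic backpack, adding flickering glows and lights.
# EyeGlowStack pins two particles to skeletal mesh bones or sockets by partition index, layering three sprite renderers and a light with orange/blue variant colors.
# An orange glow-and-light emitter and a pink-purple NeonGlow emitter with mesh, sprite and light renderers sit at fixed offsets.
# Each emitter burst-spawns particles and derives its flicker from a Waveform module.
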NiagaraSystem /BRCosmetics/Accessories/FORT_Backpacks/Backpack_F_MED_StrategicSpur/FX/NS_Backpack_StrategicSpur_Idle
dependencies {
 usedDataInterfaces [&NiagaraDataInterfaceSkeletalMesh]
 usedDynamicInputs []
 usedModules [
  &FN_VariantManager
  &self:SystemSpawnScript.SystemScriptSource.SystemScriptGraph.NiagaraNodeAssignment_0.SetVariables_#1
  &SystemState
 ]
}
emitters {
 Default_LightAndGlowFlicker_Loop {
  emitterUpdateScript [
   &EmitterState(
    Inactive Response=::NewEnumerator1
    Life Cycle Mode=::NewEnumerator1
    Loop Behavior=::NewEnumerator1
    Loop Duration Mode=::NewEnumerator1
    UseLoopDelay=false
   )
   &SpawnBurst_Instantaneous(Use Spawn Probability=false, UseLoopCountLimit=false)
   /Niagara/Noise/Waveform(
    EnableExponent1=false
    GlobalClampMax=false
    GlobalClampMin=false
    UseInput1=false
    UsePhase1=false
    "Waveform [1]":ENiagara_Waveforms=::NewEnumerator12
    "[1] Frequency"=7
   )
   /SetVariables(Emitter.Flicker:NiagaraFloat=&Add_Float(A=&Multiply_Float(A=$Emitter.Waveform.Waveform, B=0.5), B=0.5)) {isSetParameters=true, module=SetVariables_#2}
  ]
  particleSpawnScript [
   &InitializeParticle(
    Color=<rgba 1 0.1831 0 0.7>
    Position Offset=<0 -12.9663 -4.4045>
    Position Offset Coordinate Space=::Local
    Sprite Size Mode=::NewEnumerator3
    Uniform Sprite Size=14.6378
    UsePositionOffset=true
    Write Material Random=true
   )
   &DynamicMaterialParameters(
    Index 0 Param 1=0.2
    Index 0 Param 2=1
    Module.Param0WriteEnabled=true
    Module.Param1WriteEnabled=true
    Module.Param2WriteEnabled=true
    Module.Param3WriteEnabled=true
   )
  ]
  particleUpdateScript [
   &ParticleState(Kill Particles When Lifetime Has Elapsed=false, Let Infinitely Lived Particles Die When Emitter Deactivates=true)
   &ScaleColor(Scale Alpha=&Multiply_Float(A=$Emitter.Flicker, B=0.3), ScaleA=true, ScaleRGB=true)
   &FN_Light_Attributes(
    Color=$Particles.Color
    Create Color Attribute=true
    Create Exponent Attribute=true
    Create Radius Attribute=true
    Exponent=1
    Radius=8
    Scale Color=&Multiply_Float(A=7, B=$Emitter.Flicker)
   )
  ]
  rendererClasses [&NiagaraSpriteRendererProperties, &NiagaraLightRendererProperties]
  renderers [
   {rendererClass=&NiagaraSpriteRendererProperties}
   {rendererClass=&NiagaraLightRendererProperties, rendererIndex=1}
  ]
 }
 EyeGlowStack {
  emitterUpdateScript [
   &EmitterState(
    Inactive Response=::NewEnumerator1
    Life Cycle Mode=::NewEnumerator1
    Loop Behavior=::NewEnumerator1
    Loop Duration Mode=::NewEnumerator1
    UseLoopDelay=true
   )
   &Waveform(
    EnableExponent1=false
    GlobalClampMax=false
    GlobalClampMin=false
    UseInput1=false
    UsePhase1=false
    "Waveform [1]":ENiagara_Waveforms=::NewEnumerator12
    "[1] Amplitude Min/Max":Vector2f=<0.7 1>
    "[1] Frequency"=7
   )
   /SetVariables(
    Emitter.Flicker:NiagaraFloat=&Add_Float(A=&Multiply_Float(A=$Emitter.Waveform.Waveform, B=0.5), B=0.7)
    Emitter.FlickerHeavy:NiagaraFloat=&Add_Float(A=&Multiply_Float(A=$Emitter.Waveform.Waveform, B=2), B=0.8)
   ) {isSetParameters=true, module=SetVariables_#2}
   &SpawnBurst_Instantaneous(Spawn Count=2, Use Spawn Probability=false, UseLoopCountLimit=true)
   /SetVariables(
    Emitter.CoreGlowColor:LinearColor=&MultiplyLinearColor(
     Linear Color=&VariantColorSelector(Variant Amount=0, Original Color=<rgba 1 0.3901 0 1>, Variant 1 Color=<rgba 0 0.3873 1 1>, Variant 2 Color=?)
     Scale Factor=&MakeVector4_FromFloat(Float=&Multiply_Float(A=1, B=$Emitter.Flicker))
    )
    Emitter.CoreGlowSize:Vector2f=&Vector2DFromFloat(Value=1.4)
    Emitter.DarkGlowColor:LinearColor=&MultiplyLinearColor(
     Linear Color=&VariantColorSelector(Variant Amount=0, Original Color=<rgba 0.1812 0.0645 0 1.5>, Variant 1 Color=<rgba 0 0.074 0.1812 1.5>, Variant 2 Color=?)
     Scale Factor=&MakeVector4_FromFloat(Float=&Multiply_Float(A=1, B=$Emitter.Flicker))
    )
    Emitter.DarkGlowSize:Vector2f=&Vector2DFromFloat(Value=8.0622)
    Emitter.DepthFade=1
    Emitter.DiffuseGlowColor:LinearColor=&MultiplyLinearColor(
     Linear Color=&VariantColorSelector(Variant Amount=0, Original Color=<rgba 2 0.0814 0 1>, Variant 1 Color=<rgba 0 2.4025 4 1>, Variant 2 Color=?)
     Scale Factor=&MakeVector4_FromFloat(Float=&Multiply_Float(A=1, B=$Emitter.Flicker))
    )
    Emitter.DiffuseGlowSize:Vector2f=&Vector2DFromFloat(Value=3.4401)
    Emitter.EyeGlowCameraOffset=1
    Emitter.LightExponent=5
    Emitter.LightGlowColor:LinearColor=&MultiplyLinearColor(
     Linear Color=&VariantColorSelector(Variant Amount=0, Original Color=<rgba 1.5 0.1087 0 1>, Variant 1 Color=<rgba 0.0352 0.1636 1 1>, Variant 2 Color=?)
     Scale Factor=&MakeVector4_FromFloat(Float=&Multiply_Float(A=7, B=$Emitter.FlickerHeavy))
    )
    Emitter.LightRadius=3
   ) {isSetParameters=true}
  ]
  particleSpawnScript [
   &PartitionParticles()
   &InitializeParticle(Color=<rgba 1 0.3963 0 0.7>, Sprite Size Mode=::NewEnumerator3, Uniform Sprite Size=120, UsePositionOffset=false, Write Material Random=true)
   &DynamicMaterialParameters(
    Index 0 Param 1=0.2
    Index 0 Param 2=1
    Module.Param0WriteEnabled=true
    Module.Param1WriteEnabled=true
    Module.Param2WriteEnabled=true
    Module.Param3WriteEnabled=true
   )
  ]
  particleUpdateScript [
   &ParticleState(Kill Particles When Lifetime Has Elapsed=false, Let Infinitely Lived Particles Die When Emitter Deactivates=true)
   &SkeletalMeshLocation(
    Bone / Socket Index=$Particles.PartitionParticles.Partition
    MaskSpawnGroup=false
    Mesh Sampling Type=::NewEnumerator5
    Skeletal Mesh=$System.SkeletalMesh
    Socket Sampling Mode=::NewEnumerator5
   )
   &ScaleColor(Scale Alpha=&Multiply_Float(A=$Emitter.Flicker, B=0.3), ScaleA=true, ScaleRGB=true)
   &FN_Light_Attributes(
    Color=$Particles.Color
    Create Color Attribute=true
    Create Exponent Attribute=true
    Create Position Attribute=true
    Create Radius Attribute=true
    Position=&AddVectorToPosition(Position=$Particles.Position, Vector=<0.1 0 0>)
    Radius=4
    Scale Color=&Multiply_Float(A=7, B=$Emitter.Flicker)
   )
  ]
  rendererClasses [&NiagaraSpriteRendererProperties, &NiagaraLightRendererProperties]
  renderers [
   {rendererClass=&NiagaraSpriteRendererProperties}
   {rendererClass=&NiagaraSpriteRendererProperties, rendererIndex=1}
   {rendererClass=&NiagaraSpriteRendererProperties, rendererIndex=2}
   {rendererClass=&NiagaraLightRendererProperties, rendererIndex=3}
  ]
 }
 NeonGlow {
  emitterUpdateScript [
   &EmitterState()
   &SpawnBurst_Instantaneous(Use Spawn Probability=false, UseLoopCountLimit=true)
   &Waveform(
    EnableExponent1=false
    GlobalClampMax=false
    GlobalClampMin=false
    UseInput1=false
    UsePhase1=false
    "Waveform [1]":ENiagara_Waveforms=::NewEnumerator12
    "[1] Frequency"=10
   )
   /SetVariables(Emitter.Flicker:NiagaraFloat=&Add_Float(A=&Multiply_Float(A=$Emitter.Waveform.Waveform, B=0.5), B=1)) {isSetParameters=true}
  ]
  particleSpawnScript [
   &InitializeParticle(
    Color=<rgba 1 0.2825 0.6135 1>
    Lifetime=$Emitter.CurrentLoopDuration
    Position Offset Coordinate Space=::Local
    Sprite Size Mode=::NewEnumerator3
    Uniform Sprite Size=33
    UsePositionOffset=true
    Write Material Random=true
   )
   &FN_Sprite_Attributes(
    Alignment:Vector3f=<0 0 1>
    Color=<rgba 0.8059 0.107 1 1>
    Create Alignment Attribute=true
    Create Color Attribute=true
    Create Facing Attribute=true
    Create Position Attribute=true
    Facing:Vector3f=<1 0 0>
    Position:NiagaraPosition=<0 -11.1334 10.0529>
    Write to Module Namespace=true
   ) {module=FN_Sprite_Attributes001}
   &FN_Sprite_Attributes(
    Alignment:Vector3f=<0 0 1>
    Color=<rgba 1.6253 0 2 1>
    Create Alignment Attribute=true
    Create Color Attribute=true
    Create Facing Attribute=true
    Create Position Attribute=true
    Create SpriteSize Attribute=true
    Facing:Vector3f=<0 -1 0>
    Position:NiagaraPosition=<0 -8.6136 10.5466>
    SpriteSize=<39.6344 35.7294>
    Write to Module Namespace=true
   ) {module=FN_Sprite_Attributes002}
  ]
  particleUpdateScript [
   &FN_Sprite_Attributes(Color=<rgba 0.187 0 1 0.15>, Create Color Attribute=true, Write to Module Namespace=true)
   &ParticleState(Kill Particles When Lifetime Has Elapsed=false)
   &FN_Light_Attributes(
    Color=$Particles.Color
    Create Color Attribute=true
    Create Exponent Attribute=true
    Create Position Attribute=true
    Create Radius Attribute=true
    Exponent=3
    Position=<0 -19.0869 15>
    Radius=20
    Scale Color=&Multiply_Float(A=5, B=$Emitter.Flicker)
   )
   &SolveForcesAndVelocity(Clamp Velocity=false, Limit Acceleration=false)
  ]
  rendererClasses [
   &NiagaraMeshRendererProperties
   &NiagaraLightRendererProperties
   &NiagaraSpriteRendererProperties
  ]
  renderers [
   {rendererClass=&NiagaraMeshRendererProperties, rendererIndex=0}
   {rendererClass=&NiagaraMeshRendererProperties, rendererIndex=1}
   {rendererClass=&NiagaraLightRendererProperties, rendererIndex=2}
   {rendererClass=&NiagaraSpriteRendererProperties, rendererIndex=3}
   {rendererClass=&NiagaraSpriteRendererProperties, rendererIndex=4}
  ]
 }
}
summary {
 systemName=NS_Backpack_StrategicSpur_Idle
 emitters [
  {
   emitterName=NeonGlow
   rendererClasses [
    &NiagaraMeshRendererProperties
    &NiagaraLightRendererProperties
    &NiagaraSpriteRendererProperties
   ]
  }
  {emitterName=EyeGlowStack, rendererClasses [&NiagaraSpriteRendererProperties, &NiagaraLightRendererProperties]}
  {
   emitterName=Default_LightAndGlowFlicker_Loop
   rendererClasses [&NiagaraSpriteRendererProperties, &NiagaraLightRendererProperties]
  }
 ]
 userVariables [
  User.bIsFrontEnd {type=NiagaraBool}
  User.bIsFrontEndPreview {type=NiagaraBool}
  User.Variant {type=NiagaraFloat, defaultValue=0, description=""}
 ]
}
